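; Cosmetic lantern actor for the research table in Season 23 Part 04 Quest 02, launched as a projectile on spawn.
; BeginPlay sets horizontal velocity from the lantern's forward vector while a timeline drives its uniform scale.
; After a delay, a second timeline plays in reverse to rescale it, then the actor destroys itself.
; The construction script offsets the capsule and sets the lantern's starting scale; the custom impact event is empty.
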(event EventBeginPlay
  (bind (_returnvalue_x _returnvalue_y) (Math|Vector|BreakVector (Transformation|GetForwardVector (Variables|Default|Getlantern))))
  (|CallParentFunction)
  (Projectile|SetVelocity (Variables|Default|GetFortProjectileMovement) (Math|Vector|MakeVector (* _returnvalue_x 6e+02) (* _returnvalue_y 6e+02)))
  (bind (_direction _grow) (Timeline|Timeline PlayFromStart)
    (:Update
      (Transformation|SetWorldScale3D (Variables|Default|Getlantern) (Math|Conversions|ToVector(Float) _grow)))
    (:Finished
      (Utilities|FlowControl|Delay 2e+01)
      (bind (_direction_2 _grow_2) (Timeline|Timeline_0 ReverseFromEnd)
        (:Update
          (Transformation|SetWorldScale3D (Variables|Default|Getlantern) (Math|Conversions|ToVector(Float) _grow_2)))
        (:Finished
          (Actor|DestroyActor))))))

(event Custom|CustomEvent
  (params (ImpactResult :type Hit%20Result%20Structure%20%28by%20ref%29) (ImpactVelocity :type Vector%20%28by%20ref%29)))

(fn ConstructionScript
  (|CallParentFunction)
  (Transformation|SetRelativeLocation (Variables|Default|GetCapsule) (Math|Vector|MakeVector (+ (.x (Variables|/Script/Engine.SceneComponent|GetRelativeLocation (Variables|Default|GetCapsule))) 0) (+ (.y (Variables|/Script/Engine.SceneComponent|GetRelativeLocation (Variables|Default|GetCapsule))) -2.0) (+ (.z (Variables|/Script/Engine.SceneComponent|GetRelativeLocation (Variables|Default|GetCapsule))) 145.0)) false false)
  (Transformation|SetWorldScale3D (Variables|Default|Getlantern) "0.400000,0.400000,0.400000"))
